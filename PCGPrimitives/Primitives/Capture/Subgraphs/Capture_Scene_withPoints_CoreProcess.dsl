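; PCG subgraph that scene-captures from one source point, or a synthetic origin point, and turns the result into world-space points.
; Input bounds or explicit extents, local offset, rotation and optional forward orientation define the capture volume.
; GPU depth capture, plus optional normal and base color, uses texel size clamped for a 2048 maximum resolution.
; A custom HLSL generator projects samples, culls by depth, writes rotation and color; debug planes are optional.
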(graph Capture_Scene_withPoints_CoreProcess
  (params (UseInputBoundsAsCaptureSource :type Bool :default true) (CaptureFromForwardVector :type Bool) (CaptureExtents :type Vector2 :default (256.0 256.0)) (CaptureTexelSize :type Double :default 1.0) (SampleSize :type Double :default 1e+01) (MaxDepth :type Double :default 256.0) (LocalPositionOffset :type Vector :default (0.0 0.0 1e+02)) (LocalRotation :type Rotator) (CaptureColor :type Bool :default true) (CaptureNormal :type Bool :default true) (CaptureOnlyIncludedTags :type Bool) (IncludedTags :type (Array Name)) (ExcludedTags :type (Array Name)) (ExcludePCGGeneratedContent :type Bool) (DebugCapturePlanes :type Bool :default true))
  (bind FilterDataByType_2 (Filter%20Data%20By%20Type :TargetType (struct :Ids [(struct :Struct /Script/PCG.PCGDataTypeInfoPoint)])))
  (bind Branch_0 (Branch :bOutputToB-param DebugCapturePlanes))
  ;[0 0 0 0.5 12] Primary/Secondary Inputs
  (bind FilterDataByTag_0 (Filter%20Data%20By%20Tag :SelectedTags "Input:0"))
  ;[0 0 0 0.5] Core Process
  (bind AttributeMathsOp_2 (Attribute%20Maths%20Op :Operation Divide :InA CaptureExtents))
  (bind DataNum_0 (Data%20Count))
  (bind Branch_1 (Branch)) ; If no input point exists, \ncreate a synthetic source point at the origin.
  (bind CreatePoints_0 (Create%20Points :CoordinateSpace LocalComponent :PointsToCreate [(struct :BoundsMin (0.0 0.0 0.0) :BoundsMax (0.0 0.0 0.0) :Steepness 1.0)]))
  (bind Branch_6 (Branch :bOutputToB-param CaptureFromForwardVector))
  (bind AttributeRotatorOp_2 (Attribute%20Rotator%20Op :InputSource1 $Rotation :InputSource2 $Rotation))
  (bind Branch_5 (Branch :bOutputToB-param CaptureFromForwardVector)) ; Branch on CaptureFromForwardVector to apply a 90 degree rotation \nso the capture shoots forward (+X) instead of down (-Z).
  (bind AttributeRotatorOp_1 (Attribute%20Rotator%20Op :InputSource1 $Rotation :InputSource2 $Rotation))
  (bind Branch_4 (Branch :bOutputToB-param UseInputBoundsAsCaptureSource))
  (bind CreatePoints_1 (Create%20Points :PointsToCreate [(struct :Transform (struct :Rotation (-8.726646259971638e-08 -0.0 0.0 0.9999999999999961)) :BoundsMin (0.0 0.0 0.0) :BoundsMax (0.0 0.0 0.0) :Steepness 1.0)]))
  (bind AttributeTransformOp_0 (Attribute%20Transform%20Op :InputSource1 $Transform :InputSource2 $Transform))
  (bind MakeTransformAttribute_0 (Make%20Transform%20Attribute :Rotation LocalRotation :Translation LocalPositionOffset)) ; Build the capture source transform: \napply LocalPositionOffset and LocalRotation on top of the input point transform
  (bind AttributeTransformOp_1 (Attribute%20Transform%20Op :InputSource1 @Last :InputSource2 $Transform))
  (bind AttributeMathsOp_1 (Attribute%20Maths%20Op :InputSource1 $Extents.xy :Operation Set :OutputTarget @Source))
  (bind AttributeMathsOp_3 (Attribute%20Maths%20Op :InputSource1 $Extents.z :Operation Set :InB MaxDepth)) ; Override the point Z extents with MaxDepth\nto define how far the capture shoots before culling.
  (bind FilterDataByType_1 (Filter%20Data%20By%20Type :TargetType (struct :Ids [(struct :Struct /Script/PCG.PCGDataTypeInfoPoint)])))
  (bind ExtractAttribute_4 (Extract%20Attribute%20at%20Index :bForceOutputAttributeToBeInElementsDomain true :InputSource $Extents.z))
  (bind AttributeMathsOp_10 (Attribute%20Maths%20Op :InputSource1 LocalPositionOffset.Position.Z))
  (bind AttributeMathsOp_11 (Attribute%20Maths%20Op :InputSource1 $Extents.z :Operation Multiply))
  (bind AttributeTransformOp_2 (Attribute%20Transform%20Op :InputSource1 LocalPositionOffset :InputSource2 $Transform))
  (bind Gather_0 (Gather))
  (bind ExtractAttribute_2 (Extract%20Attribute%20at%20Index :bForceOutputAttributeToBeInElementsDomain true :InputSource $Extents :OutputAttributeName SampleExtents)) ; Compute the total number of output samples
  (bind AttributeMathsOp_6 (Attribute%20Maths%20Op :Operation Multiply))
  (bind AttributeMathsOp_8 (Attribute%20Maths%20Op :Operation Divide :InB SampleSize))
  (bind AttributeMathsOp_16 (Attribute%20Maths%20Op :Operation Ceil))
  (bind AttributeMathsOp_7 (Attribute%20Maths%20Op :InputSource1 SampleExtents.x :InputSource2 SampleExtents.y :Operation Multiply :OutputTarget NumElements))
  (bind ExtractAttribute_3 (Extract%20Attribute%20at%20Index :bForceOutputAttributeToBeInElementsDomain true :InputSource NumElements))
  (bind AttributeCast_0 (Attribute%20Cast :OutputType Integer64))
  (bind AttributeCast_1 (Attribute%20Cast :OutputType Integer32 :In CaptureNormal))
  (bind AttributeCast_2 (Attribute%20Cast :OutputType Integer32 :In CaptureColor))
  (bind MergeAttributes_0 (Merge%20Attributes :In SampleSize))
  (bind CopyAttributes_0 (Copy%20Attributes :bCopyAllAttributes true :InputSource @Last)) ; Pack SampleSize, CaptureNormal flag, and CaptureColor flag \ninto a single param dataset to pass into the HLSL kernel as SampleAttributes.
  (bind ExtractAttribute_0 (Extract%20Attribute%20at%20Index :bForceOutputAttributeToBeInElementsDomain true :InputSource ClampedCaptureTexelSize))
  (bind ExtractAttribute_8 (Extract%20Attribute%20at%20Index :bForceOutputAttributeToBeInElementsDomain true :InputSource $Position :OutputAttributeName CapturePosition))
  (bind ExtractAttribute_9 (Extract%20Attribute%20at%20Index :bForceOutputAttributeToBeInElementsDomain true :InputSource $Rotation :OutputAttributeName CaptureRotation))
  (bind ExtractAttribute_10 (Extract%20Attribute%20at%20Index :bForceOutputAttributeToBeInElementsDomain true :InputSource $Extents :OutputAttributeName CaptureExtents))
  (bind Branch_2 (Branch :bOutputToB-param CaptureNormal :In CaptureNormal)) ; Conditionally run the Normal scene capture
  (bind Branch_3 (Branch :bOutputToB-param CaptureColor :In CaptureColor)) ; Conditionally run the Base Color scene capture
  ;;[0 0 0 0.5] Clamping Capture Texel Size to output a scene capture of 2048x2048 max based on the capture extents.
  (bind ExtractAttribute_1 (Extract%20Attribute%20at%20Index :bForceOutputAttributeToBeInElementsDomain true :InputSource $Extents.xy :OutputAttributeName CaptureExtents))
  (bind AttributeMathsOp_9 (Attribute%20Maths%20Op :Operation Multiply))
  (bind Select_0 (Select :bUseInputB-param UseInputBoundsAsCaptureSource :Input%20A CaptureExtents)) ; Compute the clamped capture texel size
  (bind AttributeMathsOp_12 (Attribute%20Maths%20Op :InputSource1 @Last.x :InputSource2 @Last.y :Operation Max :OutputTarget LongestComponent))
  (bind AttributeMathsOp_13 (Attribute%20Maths%20Op :InputSource1 LongestComponent :Operation Divide :OutputTarget ComputeSceneCaptureRes :InB CaptureTexelSize))
  (bind AttributeMathsOp_14 (Attribute%20Maths%20Op :InputSource1 ComputeSceneCaptureRes :Operation Min :OutputTarget ClampedRes))
  (bind AttributeMathsOp_15 (Attribute%20Maths%20Op :InputSource1 LongestComponent :InputSource2 ClampedRes :Operation Divide :OutputTarget ClampedCaptureTexelSize))
  ;;[0 0 0 0.5] GPU Scene Captures
  (bind SceneCapture_1 (Scene%20Capture :bCaptureFromTopOfExtents false :bExcludePCGContent false :CaptureSource SCS_SceneDepth :Filter Point :OrientationMode Explicit :PixelFormat RTF_R16f :TexelSize 1.0 :bExcludePCGContent-param ExcludePCGGeneratedContent :bOnlyCaptureContentMatchingIncludedTags-param CaptureOnlyIncludedTags :ExcludedTags-param ExcludedTags :IncludedTags-param IncludedTags)) ; Depth capture always runs - \nit drives point position and is required regardless of optional captures.
  (bind SceneCapture_2 (Scene%20Capture :bCaptureFromTopOfExtents false :bExcludePCGContent false :CaptureSource SCS_Normal :Filter Point :OrientationMode Explicit :TexelSize 1.0 :bExcludePCGContent-param ExcludePCGGeneratedContent :bOnlyCaptureContentMatchingIncludedTags-param CaptureOnlyIncludedTags :ExcludedTags-param ExcludedTags :IncludedTags-param IncludedTags)) ; Normal scene capture only when CaptureNormal is enabled.
  (bind SceneCapture_3 (Scene%20Capture :bCaptureFromTopOfExtents false :bExcludePCGContent false :CaptureSource SCS_BaseColor :Filter Point :OrientationMode Explicit :TexelSize 1.0 :bExcludePCGContent-param ExcludePCGGeneratedContent :bOnlyCaptureContentMatchingIncludedTags-param CaptureOnlyIncludedTags :ExcludedTags-param ExcludedTags :IncludedTags-param IncludedTags)) ; Base Color scene capture only when CaptureColor is enabled.
  ;;[0 0 0 0.5] Sample Scene Capture(s) and output points
  (bind CustomHLSL_0 (Custom%20HLSL :InputPins [(struct :Label CaptureDepth :AllowedTypes (struct :Ids [(struct :Struct /Script/PCG.PCGDataTypeInfoTexture2DSingleBase)]) :bAllowMultipleData false :bAllowMultipleConnections false) (struct :Label CaptureColor :AllowedTypes (struct :Ids [(struct :Struct /Script/PCG.PCGDataTypeInfoTexture2DSingleBase)]) :bAllowMultipleData false :bAllowMultipleConnections false) (struct :Label CaptureNormal :AllowedTypes (struct :Ids [(struct :Struct /Script/PCG.PCGDataTypeInfoTexture2DSingleBase)]) :bAllowMultipleData false :bAllowMultipleConnections false) (struct :Label Source :AllowedTypes (struct :Ids [(struct :Struct /Script/PCG.PCGDataTypeInfoPoint)]) :bAllowMultipleConnections false) (struct :Label SampleAttributes :AllowedTypes (struct :Ids [(struct :Struct /Script/PCG.PCGDataTypeInfoParam)]) :bAllowMultipleConnections false)] :KernelType PointGenerator :NumElements 64 :OutputPins [(struct :PropertiesGPU (struct :CreatedKernelAttributeKeys [(struct :Identifier (struct :Name Depth) :Name Depth)]) :Label Out :AllowedTypes (struct :Ids [(struct :Struct /Script/PCG.PCGDataTypeInfoPoint)]))])) ; GPU kernel: reads depth (and optionally normal/color) textures, projects each texel back to world space using the source transform, \nculls by MaxDepth, and outputs one point per sample with position, density, optional rotation (from normal), and optional $Color.
  (bind MutateSeed_0 (Mutate%20Seed))
  ;[1 0 0 0.5] single input, single point source supported
  (bind FilterDataByIndex_0 (Filter%20Data%20By%20Index :SelectedIndices "0"))
  (bind FilterElementsByIndex_0 (Filter%20Elements%20By%20Index :bSelectIndicesByInput false :SelectedIndices "0"))
  ;[0 0 0 0.5] Debug Capture Source and Max Planes
  (bind ResetPointCenter_1 (Reset%20Point%20Center))
  (bind AttributeMathsOp_0 (Attribute%20Maths%20Op :InputSource1 $Position))
  (bind CreateConstant_2 (Create%20Constant :AttributeTypes (struct :Type Vector :VectorValue (1.0 1.0 0.0001))))
  (bind AttributeMathsOp_18 (Attribute%20Maths%20Op :InputSource1 $Scale :Operation Set))
  (bind Debug_1 (Debug))
  (bind ResetPointCenter_0 (Reset%20Point%20Center :PointCenterLocation (0.5 0.5 0.0)))
  (bind AttributeMathsOp_4 (Attribute%20Maths%20Op :InputSource1 $Position :Operation Subtract))
  (bind AttributeMathsOp_19 (Attribute%20Maths%20Op :InputSource1 $Scale :Operation Set))
  (bind Debug_2 (Debug))
  (connect ExtractAttribute_3 Out AttributeCast_0 In)
  (connect ResetPointCenter_1 Out AttributeMathsOp_0 InA)
  (connect AttributeTransformOp_1 Out AttributeMathsOp_1 InA)
  (connect AttributeMathsOp_2 Out AttributeMathsOp_1 InB)
  (connect AttributeMathsOp_1 Out AttributeMathsOp_3 InA)
  (connect ResetPointCenter_0 Out AttributeMathsOp_4 InA)
  (connect ExtractAttribute_2 Out AttributeMathsOp_6 InA)
  (connect AttributeMathsOp_16 Out AttributeMathsOp_7 InA)
  (connect AttributeMathsOp_16 Out AttributeMathsOp_7 InB)
  (connect AttributeMathsOp_6 Out AttributeMathsOp_8 InA)
  (connect ExtractAttribute_1 Out AttributeMathsOp_9 InA)
  (connect MakeTransformAttribute_0 Out AttributeMathsOp_10 InA)
  (connect ExtractAttribute_4 Out AttributeMathsOp_10 InB)
  (connect FilterDataByType_1 InsideFilter AttributeMathsOp_11 InA)
  (connect Select_0 Out AttributeMathsOp_12 InA)
  (connect Select_0 Out AttributeMathsOp_12 InB)
  (connect AttributeMathsOp_12 Out AttributeMathsOp_13 InA)
  (connect AttributeMathsOp_13 Out AttributeMathsOp_14 InA)
  (connect AttributeMathsOp_14 Out AttributeMathsOp_15 InA)
  (connect AttributeMathsOp_14 Out AttributeMathsOp_15 InB)
  (connect AttributeMathsOp_8 Out AttributeMathsOp_16 In)
  (connect AttributeMathsOp_0 Out AttributeMathsOp_18 InA)
  (connect CreateConstant_2 Out AttributeMathsOp_18 InB)
  (connect AttributeMathsOp_4 Out AttributeMathsOp_19 InA)
  (connect CreateConstant_2 Out AttributeMathsOp_19 InB)
  (connect Branch_5 Output%20B AttributeRotatorOp_1 InB)
  (connect Branch_6 Output%20B AttributeRotatorOp_2 InB)
  (connect CreatePoints_1 Out AttributeTransformOp_0 InA)
  (connect Branch_4 Output%20A AttributeTransformOp_0 InB)
  (connect MakeTransformAttribute_0 Out AttributeTransformOp_1 InA)
  (connect AttributeRotatorOp_2 Out AttributeTransformOp_1 InB)
  (connect AttributeTransformOp_0 Out AttributeTransformOp_1 InB)
  (connect Branch_6 Output%20A AttributeTransformOp_1 InB)
  (connect AttributeMathsOp_10 Out AttributeTransformOp_2 InA)
  (connect AttributeMathsOp_11 Out AttributeTransformOp_2 InB)
  (connect Gather_0 Out Branch_0 In)
  (connect DataNum_0 Out Branch_1 bOutputToB)
  (connect FilterElementsByIndex_0 In%20Filter Branch_1 In)
  (connect AttributeRotatorOp_1 Out Branch_4 In)
  (connect Branch_5 Output%20A Branch_4 In)
  (connect Branch_1 Output%20B Branch_5 In)
  (connect CreatePoints_0 Out Branch_6 In)
  (connect MergeAttributes_0 Out CopyAttributes_0 Source)
  (connect ExtractAttribute_2 Out CopyAttributes_0 Target)
  (connect Branch_1 Output%20A CreatePoints_0 Execution%20Dependency)
  (connect Branch_4 Output%20A CreatePoints_1 Execution%20Dependency)
  (connect SceneCapture_3 Out CustomHLSL_0 CaptureColor)
  (connect SceneCapture_1 Out CustomHLSL_0 CaptureDepth)
  (connect SceneCapture_2 Out CustomHLSL_0 CaptureNormal)
  (connect AttributeCast_0 Out CustomHLSL_0 NumElements)
  (connect CopyAttributes_0 Out CustomHLSL_0 SampleAttributes)
  (connect Gather_0 Out CustomHLSL_0 Source)
  (connect FilterElementsByIndex_0 In%20Filter DataNum_0 In)
  (connect AttributeMathsOp_18 Out Debug_1 In)
  (connect AttributeMathsOp_19 Out Debug_2 In)
  (connect AttributeMathsOp_15 Out ExtractAttribute_0 In)
  (connect FilterDataByType_1 InsideFilter ExtractAttribute_1 In)
  (connect Gather_0 Out ExtractAttribute_2 In)
  (connect AttributeMathsOp_7 Out ExtractAttribute_3 In)
  (connect FilterDataByType_1 InsideFilter ExtractAttribute_4 In)
  (connect Gather_0 Out ExtractAttribute_8 In)
  (connect Gather_0 Out ExtractAttribute_9 In)
  (connect Gather_0 Out ExtractAttribute_10 In)
  (connect FilterDataByType_2 InsideFilter FilterDataByIndex_0 In)
  (connect DefaultInputNode In FilterDataByTag_0 In)
  (connect Branch_4 Output%20B FilterDataByType_1 In)
  (connect FilterDataByTag_0 InsideFilter FilterDataByType_2 In)
  (connect FilterDataByIndex_0 InsideFilter FilterElementsByIndex_0 In)
  (connect AttributeMathsOp_3 Out Gather_0 In)
  (connect AttributeTransformOp_2 Out Gather_0 In)
  (connect AttributeCast_1 Out MergeAttributes_0 In2)
  (connect AttributeCast_2 Out MergeAttributes_0 In3)
  (connect CustomHLSL_0 Out MutateSeed_0 In)
  (connect Branch_0 Output%20B ResetPointCenter_0 In)
  (connect Branch_0 Output%20B ResetPointCenter_1 In)
  (connect ExtractAttribute_10 Out SceneCapture_1 CaptureHalfExtents)
  (connect ExtractAttribute_8 Out SceneCapture_1 CaptureLocation)
  (connect ExtractAttribute_9 Out SceneCapture_1 CaptureRotation)
  (connect ExtractAttribute_0 Out SceneCapture_1 TexelSize)
  (connect ExtractAttribute_10 Out SceneCapture_2 CaptureHalfExtents)
  (connect ExtractAttribute_8 Out SceneCapture_2 CaptureLocation)
  (connect ExtractAttribute_9 Out SceneCapture_2 CaptureRotation)
  (connect Branch_2 Output%20B SceneCapture_2 Execution%20Dependency)
  (connect ExtractAttribute_0 Out SceneCapture_2 TexelSize)
  (connect ExtractAttribute_10 Out SceneCapture_3 CaptureHalfExtents)
  (connect ExtractAttribute_8 Out SceneCapture_3 CaptureLocation)
  (connect ExtractAttribute_9 Out SceneCapture_3 CaptureRotation)
  (connect Branch_3 Output%20B SceneCapture_3 Execution%20Dependency)
  (connect ExtractAttribute_0 Out SceneCapture_3 TexelSize)
  (connect AttributeMathsOp_9 Out Select_0 Input%20B)
  (output Out MutateSeed_0))
